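# Translucent unlit Niagara mesh-particle material for the Order Guard glider's sawblade effect, combining spinning motion blur with wind streaks.
# A panned blur texture, faded along V, forms a darkened mask that also dims the wind color.
# On UV1, two skewed, panned sawblade samples multiply and erode against a noise-varied gradient, then merge with the blur as particle-alpha-scaled opacity.
# It computes a camera-facing vertex push and uses depth fade at higher quality levels.
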Material /BRCosmetics/Gadgets/Assets/VinderTech_GliderChute/Glider_Order_Guard_Male/FX/Materials/M_Glider_OrderGuard_Sawblades
graph {
 properties {
  bUsedWithNiagaraMeshParticles=true
  blendMode=BLEND_Translucent
  pixelDepthOffsetMode=PDOM_Legacy
  shadingModel=MSM_Unlit
  usedShadingModels=Unlit
 }
 nodes [
  n1 = TextureCoordinate() {desc="1, .7", vTiling=0.6}
  n2 = ScalarParameter() {defaultValue=-0.375, parameterName=Blur Panner Speed}
  n3 = Constant()
  n4 = TextureSample(UVs=Panner(Coordinate=@n1, Speed=AppendVector(A=@n2, B=@n3)) {bFractionalPart=true}) {
   samplerType=SAMPLERTYPE_LinearGrayscale
   texture=/BRCosmetics/Gadgets/Assets/VinderTech_GliderChute/Glider_Order_Guard_Male/FX/Textures/T_Glider_OrderGuard_Sawblade_Blur
  }
  n5 = ComponentMask(None=@n1) {g=true}
  n6 = Saturate(None=Divide(A=OneMinus(None=@n5), B=ScalarParameter() {defaultValue=0.6, parameterName=Softness}) {constB=0.2})
  n7 = Multiply(
   A=Power(Base=Multiply(A=@n4.R, B=@n6), Exp=ScalarParameter() {defaultValue=2, parameterName=Motion blur Exponent}) {constExponent=1}
   B=ScalarParameter() {defaultValue=0.6, parameterName=Darkening}
  )
  MaterialFunctionCall(Emissive=Multiply(A=VectorParameter() {parameterName=Wind Color, defaultValue=<rgba 0.1 0.1 0.1 0>}.RGB, B=OneMinus(None=@n7))) {materialFunction=&MF_FX_EmissiveAdjustments}
  n8 = CameraPositionWS()
  n9 = WorldPosition()
  Multiply(A=Normalize(VectorInput=Subtract(A=@n8, B=@n9.XYZ)), B=ScalarParameter() {defaultValue=2, parameterName=Vertex Push})
  n10 = MaterialFunctionCall(Float2=TextureCoordinate() {coordinateIndex=1}) {materialFunction=&BreakOutFloat2Components}
  n11 = Constant()
  n12 = Multiply(A=Time(), B=ScalarParameter() {defaultValue=1.1, parameterName=Time Scale Wind})
  n13 = VectorParameter() {parameterName=Panner Speeds, defaultValue=<rgba -0.5 -0.2 -0.13 -0.7>}
  n14 = TextureSample(UVs=Panner(Coordinate=AppendVector(A=@n10.R, B=@n11), Time=@n12, Speed=AppendVector(A=@n13.B, B=@n13.B))) {samplerType=SAMPLERTYPE_LinearGrayscale, texture=/Engine/Functions/Engine_MaterialFunctions02/ExampleContent/Textures/LowResBlurredNoise}
  Multiply(A=@n14.R, B=@n14.R)
  n15 = Multiply(A=@n10.R, B=ScalarParameter() {defaultValue=-7, parameterName=Skew})
  n16 = Panner(Coordinate=AppendVector(A=@n10.R, B=Add(A=@n10.G, B=@n15)), Time=@n12, Speed=@n13.RGB)
  n17 = TextureSample(UVs=@n16) {
   samplerType=SAMPLERTYPE_LinearGrayscale
   texture=/BRCosmetics/Gadgets/Assets/VinderTech_GliderChute/Glider_Order_Guard_Male/FX/Textures/T_Glider_OrderGuard_Sawblade
  }
  n18 = TextureSample(UVs=Multiply(A=Add(A=@n16) {constB=0.5}) {constB=2}) {
   samplerType=SAMPLERTYPE_LinearGrayscale
   texture=&T_Glider_OrderGuard_Sawblade
  }
  n19 = Multiply(A=@n17.R, B=@n18.R)
  n20 = Power(Base=@n10.G, Exp=Multiply(A=ScalarParameter() {defaultValue=6, parameterName=Exponent Max}, B=@n14.R))
  n21 = ScalarParameter() {defaultValue=0.3, parameterName=Exponent Min}
  n22 = Saturate(None=LinearInterpolate(
    A=MaterialFunctionCall!(Gradient=@n19, Erosion Amount=@n20) {materialFunction=/BRCosmetics/Athena/Effects/Particles/Skydiving/NS_Contrail_DarkAndLight/Functions/MF_ErosionNormalized}.Result
    B=MaterialFunctionCall!(Gradient=@n19, Erosion Amount=@n20, Softness=@n21) {materialFunction=/BRCosmetics/Athena/Effects/Particles/Skydiving/NS_Contrail_DarkAndLight/Functions/MF_Erosion}.Default Range
    Alpha=ScalarParameter() {parameterName=testeee}
   ))
  n23 = ScalarParameter() {defaultValue=0.6, parameterName=Opacity Wind}
  n24 = Multiply(A=ParticleColor().A, B=Max(A=Multiply(A=@n22, B=@n23), B=Saturate(None=@n7)))
  n25 = DepthFade(Opacity=@n24, FadeDistance=ScalarParameter() {defaultValue=3, parameterName=Depth Fade Distance})
  QualitySwitch(Default=@n25, Low=@n24, High=@n25, Medium=@n24, Epic=@n25)
 ]
}
