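# Translucent unlit material for Niagara sprite sparks on the sawblades of the male Order Guard glider.
# A grayscale spark texture multiplies particle color RGB, and the result feeds the Emissive input of the shared MF_FX_EmissiveAdjustments function.
# The same texture channel multiplies particle alpha to form a mask, so the particle system controls both tint and fade.
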Material /BRCosmetics/Gadgets/Assets/VinderTech_GliderChute/Glider_Order_Guard_Male/FX/Materials/M_Glider_Orderguard_Sawblades_Sparks
graph {
 properties {bUsedWithNiagaraSprites=true, blendMode=BLEND_Translucent, pixelDepthOffsetMode=PDOM_Legacy, shadingModel=MSM_Unlit, usedShadingModels=Unlit}
 nodes [
  n1 = ParticleColor()
  n2 = TextureSample() {
   samplerType=SAMPLERTYPE_LinearGrayscale
   texture=/Game/Gadgets/Assets/VinderTech_GliderChute/Glider_Order_Guard_Male/FX/Textures/T_Glider_OrderGuard_Sawblade_Sparks
  }
  MaterialFunctionCall(Emissive=Multiply(A=@n1.RGB, B=@n2.R)) {materialFunction=&MF_FX_EmissiveAdjustments}
  Multiply(A=@n1.A, B=@n2.R)
 ]
}
