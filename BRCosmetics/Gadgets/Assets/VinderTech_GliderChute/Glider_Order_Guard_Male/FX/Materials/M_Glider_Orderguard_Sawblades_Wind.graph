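# Translucent unlit Niagara mesh-particle material for wind streaks on the Order Guard sawblade glider.
# A flat Color parameter feeds an emissive-adjustment function, separate from the alpha chain.
# For alpha, two inverted, panned sawblade texture samples multiply into an erosion function, then are shaped by sine UV fades, a view-facing normal falloff, color alpha and particle alpha.
# Additional branches compute a mirrored center-line pattern and axis-aligned fresnel, apart from that chain.
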Material /BRCosmetics/Gadgets/Assets/VinderTech_GliderChute/Glider_Order_Guard_Male/FX/Materials/M_Glider_Orderguard_Sawblades_Wind
graph {
 properties {
  bUsedWithNiagaraMeshParticles=true
  blendMode=BLEND_Translucent
  pixelDepthOffsetMode=PDOM_Legacy
  shadingModel=MSM_Unlit
  usedShadingModels=Unlit
 }
 nodes [
  ComponentMask(None=VectorParameter() {defaultValue=<rgba 2 3 0 0>}.RGB) {g=true, r=true}
  Constant4Vector() {constant=<rgba -1 1 -0.618 -1>}
  Constant() {r=1}
  n1 = MaterialFunctionCall(Float2=TextureCoordinate() {vTiling=2}) {materialFunction=&BreakOutFloat2Components}
  n2 = Subtract(A=@n1.R) {constB=0.5}
  n3 = Abs(None=@n2)
  n4 = Saturate(None=Divide(A=@n3, B=ScalarParameter() {defaultValue=0.1, parameterName=Central Line Size}))
  n5 = Ceil(None=@n2)
  n6 = AppendVector(A=@n3, B=Add(A=@n1.G, B=Divide(A=@n5)))
  n7 = VectorParameter() {parameterName=PannerSpeed, defaultValue=<rgba -0.3 0.7 -0.4 1>}
  n8 = TextureSample(UVs=Panner(Coordinate=@n6, Speed=AppendVector(A=@n7.R, B=@n7.R))) {
   samplerType=SAMPLERTYPE_LinearGrayscale
   texture=/BRCosmetics/Gadgets/Assets/VinderTech_GliderChute/Glider_Order_Guard_Male/FX/Textures/T_Glider_OrderGuard_Sawblade
  }
  n9 = TextureSample(UVs=Panner(Coordinate=@n6, Speed=AppendVector(A=@n7.B, B=@n7.B))) {
   samplerType=SAMPLERTYPE_LinearGrayscale
   texture=&T_Glider_OrderGuard_Sawblade
  }
  n10 = Multiply(A=@n8.R, B=@n9.R)
  MaterialFunctionCall!(Blend=OneMinus(None=@n4), Base=@n10) {materialFunction=/Engine/Functions/Engine_MaterialFunctions03/Blends/Blend_Screen}
  n11 = VectorParameter() {parameterName=Color, defaultValue=<rgba 0.3 0.3 0.3 0.5>}
  MaterialFunctionCall(Emissive=@n11.RGB) {materialFunction=&MF_FX_EmissiveAdjustments}
  n12 = DotProduct(A=VertexNormalWS(), B=CameraVectorWS())
  n13 = CrossProduct(A=VertexTangentWS(), B=VertexNormalWS())
  n14 = ScalarParameter() {defaultValue=6, parameterName=Fresnel Exponent}
  Min(
   A=@n12
   B=MaterialFunctionCall(Axis=Normalize(VectorInput=@n13), Exponent=@n14) {materialFunction=&AxisAlignedFresnel}.Fresnel
  )
  Multiply(A=@n4, B=@n10)
  n15 = Multiply(
   A=TextureCoordinate()
   B=AppendVector(A=ScalarParameter() {defaultValue=2, parameterName=UV Scale U}, B=ScalarParameter() {defaultValue=3, parameterName=UV Scale V})
  )
  n16 = VectorParameter() {parameterName=Noise Pan Speeds, defaultValue=<rgba 0.05 4.23 -0.09 3.17>}
  n17 = TextureSample(UVs=Panner(Coordinate=@n15, Speed=ComponentMask(None=@n16.RGB) {g=true, r=true})) {
   samplerType=SAMPLERTYPE_LinearGrayscale
   texture=&T_Glider_OrderGuard_Sawblade
  }
  n18 = OneMinus(None=@n17.R)
  n19 = AppendVector(A=@n16.B, B=@n16.B)
  n20 = TextureSample(UVs=Add(A=Panner(Coordinate=@n15, Speed=@n19)) {constB=0.5}) {
   samplerType=SAMPLERTYPE_LinearGrayscale
   texture=&T_Glider_OrderGuard_Sawblade
  }
  n21 = OneMinus(None=@n20.R)
  n22 = TextureCoordinate()
  n23 = Sine(None=Divide(A=ComponentMask(None=@n22) {r=true}))
  n24 = ScalarParameter() {defaultValue=3, parameterName=Tips Fade Exponent}
  n25 = Multiply(
   A=MaterialFunctionCall!(
    Gradient=Multiply(A=@n18, B=@n21)
    Erosion Amount=ScalarParameter() {defaultValue=0.4, parameterName=Erosion Amount}
    Softness=ScalarParameter() {defaultValue=0.3, parameterName=Erosion Softness}
   ) {materialFunction=/BRCosmetics/Athena/Effects/Particles/Skydiving/NS_Contrail_DarkAndLight/Functions/MF_Erosion}.Default Range
   B=Multiply(A=Power(Base=@n23, Exp=@n24), B=Power(Base=@n12, Exp=@n14))
  )
  n26 = Divide(A=ComponentMask(None=TextureCoordinate()) {g=true})
  n27 = Saturate(None=Divide(A=Sine(None=@n26), B=ScalarParameter() {defaultValue=0.4, parameterName=Opacity Mask Softness}))
  Multiply(A=ParticleColor().A, B=Multiply(A=@n11.A, B=Multiply(A=@n25, B=@n27)))
 ]
}
