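# Niagara effect for the male Order Guard glider, combining a mesh, electric arcs, sparks, light-emitting pops and smoke, with spark sounds.
# A system timer value shifts Sparks and Spark_Pop from bursts to continuous spawning, quickens Electricity loops and sustains Smoke after its initial puff.
# Electricity steps through seven fixed positions in User.LightningOrder sequence, using spawn group as mesh index.
# Sparks and smoke trail against owner velocity; FN_Debug is disabled.
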NiagaraSystem /BRCosmetics/Gadgets/Assets/VinderTech_GliderChute/Glider_Order_Guard_Male/FX/NS_Glider_OrderGuard
dependencies {
 usedDataInterfaces [
  &NiagaraDataInterfaceArrayFloat3
  &NiagaraDataInterfaceArrayFloat
  &NiagaraDataInterfaceAudioPlayer
  &NiagaraDataInterfaceCurve
  &NiagaraDataInterfaceVector2DCurve
  &NiagaraDataInterfaceColorCurve
 ]
 usedDynamicInputs [&MakeCustomFloatFromBool]
 usedModules [&SystemState, &FN_Integrator]
}
emitters {
 Electricity {
  emitterUpdateScript [
   &EmitterState(
    Delay First Loop Only=true
    Inactive Response=::NewEnumerator1
    Life Cycle Mode=::NewEnumerator1
    Loop Delay=3.4
    Loop Duration=&RandomRangeFloat(
     Minimum=0.066
     Maximum=&Lerp_Float(A=1.3, B=0.233, Alpha=$System.FN_Timer001.Value)
     Evaluation Type=::NewEnumerator1
    )
    Recalculate Duration Each Loop=true
    UseLoopDelay=true
   )
   &SpawnBurst_Instantaneous(
    Spawn Group:NiagaraInt32=&SelectIntFromArray(Direct Array Index=$Emitter.LoopCount, Direct Array Mode=::NewEnumerator1, Int Selection Array=$User.LightningOrder)
    Use Spawn Probability=false
    UseLoopCountLimit=false
   )
  ]
  particleSpawnScript [
   &InitializeParticle(
    Color=<rgba 0 2.6667 4 0.013>
    Lifetime=0.666
    Position=&SelectVectorFromArray(
     Direct Array Index=&SpawnGroup()
     Direct Array Mode=::NewEnumerator1
     Vector Selection Array=DataInterface() {propertyValues {FloatData [<-17 88 -26>, <22 88 -22>, <20 33 -40>, <-10 4 -47>, <13 -70 -37>, <-35 -127 -23>, <15 -149 -33>]}}
    )
    Position Mode=::NewEnumerator1
    Sprite Size Mode=::NewEnumerator3
    Uniform Sprite Size=44
    UsePositionOffset=false
    Write Material Random=true
   )
   /SetVariables(
    Particles.FakeLightSize:Vector2f=&Vector2DFromFloat(Value=66)
    Particles.MeshIndex=&SpawnGroup()
   ) {isSetParameters=true}
   &AddVelocity(
    Cone Axis=<0 1 0>
    Module.UseConeFalloff=false
    Module.UseVelDistribution=false
    Override Seed=false
    Velocity Mode=::NewEnumerator2
    Velocity Speed=666
   )
   &ApplyInitialForces()
  ]
  particleUpdateScript [
   &ParticleState()
   &ScaleColor(
    Scale RGB=&VectorFromFloat(Value=&FloatFromCurve(FloatCurve=DataInterface() {propertyValues {Curve {keys [
          <at 0 0 RCIM_Constant>
          <at 0.027 3 RCIM_Constant>
          {arriveTangent=-1.3556, interpMode=RCIM_Cubic, leaveTangent=-1.3556, tangentMode=RCTM_User, time=0.0555, value=0.25}
          {arriveTangentWeight=0.3203, interpMode=RCIM_Cubic, tangentMode=RCTM_User, time=0.7, value=0}
         ]}}}))
    ScaleA=false
    ScaleRGB=true
   )
  ]
  rendererClasses [&NiagaraMeshRendererProperties, &NiagaraSpriteRendererProperties]
  renderers [
   {rendererClass=&NiagaraMeshRendererProperties, rendererIndex=0}
   {rendererClass=&NiagaraSpriteRendererProperties, rendererIndex=1}
   {rendererClass=&NiagaraSpriteRendererProperties, rendererIndex=2}
  ]
 }
 FN_Debug {
  bEnabled=false
  emitterUpdateScript [
   &EmitterState()
   &SpawnBurst_Instantaneous(Spawn Count=1, Use Spawn Probability=false)
  ]
  particleSpawnScript [/Niagara/Modules/Spawn/Initialization/InitializeParticle(
    Lifetime=5
    Module.Write Color=true
    Module.Write Lifetime=true
    Module.Write Mass=true
    Module.Write Material Random=true
    Module.Write Position=false
    Module.Write Scale=false
    Module.Write SpriteRotation=false
    Module.Write SpriteSize=true
    Position=&SimulationPosition()
    Sprite Size=<100 100>
   )]
  particleUpdateScript [
   &ParticleState()
   &FN_Timer()
   /SetVariables(Particles.Debug:NiagaraFloat=$System.FN_Timer.Value) {isSetParameters=true}
   &DynamicMaterialParameters(
    Index 0 Param 1=$Particles.Debug
    Module.Param0WriteEnabled=true
    Module.Param1WriteEnabled=true
    Module.Param2WriteEnabled=true
    Module.Param3WriteEnabled=true
   )
   &CameraOffset(Camera Offset Amount=111)
  ]
  rendererClasses [&NiagaraSpriteRendererProperties]
  renderers [{rendererClass=&NiagaraSpriteRendererProperties}]
 }
 Meshes {
  emitterUpdateScript [
   &EmitterState(
    Inactive Response=::NewEnumerator1
    Life Cycle Mode=::NewEnumerator1
    Loop Behavior=::NewEnumerator1
    Loop Duration Mode=::NewEnumerator1
    UseLoopDelay=false
   )
   &SpawnBurst_Instantaneous(Use Spawn Probability=false, UseLoopCountLimit=false)
  ]
  particleSpawnScript [/Niagara/Modules/Spawn/Initialization/V2/InitializeParticle(Color Mode=::NewEnumerator0, Lifetime=5, UsePositionOffset=false, Write Material Random=true)]
  particleUpdateScript [
   &ParticleState(Kill Particles When Lifetime Has Elapsed=false)
   &Color(Color=&MakeLinearColorFromVectorAndFloat("Vector (RGB)"=<1 1 1>, "Float (Alpha)"=&Subtract_Float(A=$System.FN_Timer.Value, B=0.25)))
  ]
  rendererClasses [&NiagaraMeshRendererProperties]
  renderers [{rendererClass=&NiagaraMeshRendererProperties, rendererIndex=0}]
 }
 Smoke {
  emitterUpdateScript [
   &EmitterState()
   &SpawnRate(
    Spawn Probability=&Max_Float(
     A:NiagaraFloat=$System.FN_Timer001.Value
     B:NiagaraFloat=&FloatFromCurve(FloatCurve=DataInterface() {propertyValues {Curve {keys [<at 0 1 RCIM_Constant>,<at 0.1 0>]}}}, CurveIndex=$System.LoopedAge)
    )
    SpawnRate=44
    Use Spawn Probability=true
   )
  ]
  particleSpawnScript [
   &InitializeParticle(
    Color=&MakeLinearColorFromVectorAndFloat("Vector (RGB)"=<1 1 1>, "Float (Alpha)"=&Lerp_Float(A=0.15, B=0.06, Alpha=$System.FN_Timer001.Value))
    Lifetime=0.333
    Position Offset=<0 88 -7>
    Sprite Rotation Mode=::NewEnumerator2
    Sprite Size Mode=::NewEnumerator3
    Sprite UV Mode=::NewEnumerator3
    Uniform Sprite Size=77
    UsePositionOffset=true
    Write Material Random=true
   )
   &ApplyInitialForces()
   &AddVelocity(
    Cone Angle=1
    Cone Axis=<0 1 0.333>
    Module.UseConeFalloff=false
    Module.UseVelDistribution=false
    Override Seed=false
    Velocity Mode=::NewEnumerator2
    Velocity Speed=&RandomRangeFloat(Minimum=333, Maximum=333)
   )
  ]
  particleUpdateScript [
   &ParticleState()
   &ScaleSpriteSize(Uniform Curve Sprite Scale=DataInterface() {propertyValues {Curve {keys [
        {arriveTangent=0.3484, interpMode=RCIM_Cubic, leaveTangent=0.3484, tangentMode=RCTM_User, time=0, value=0.15}
        {arriveTangent=1.6092, interpMode=RCIM_Cubic, leaveTangent=1.6092, tangentMode=RCTM_User, time=1, value=1}
       ]}}})
   &DynamicMaterialParameters(
    Index 0 Param 1=&Add_Float(
     A=&Multiply_Float(A=$Particles.NormalizedAge, B=0.4)
     B=&RandomRangeFloat(Minimum=0, Maximum=1)
    )
    Index 0 Param 3=&FloatFromCurve(
     FloatCurve=DataInterface() {propertyValues {Curve {keys [
         {arriveTangent=-3.1042, interpMode=RCIM_Cubic, leaveTangent=-3.1042, tangentMode=RCTM_User, time=0.2, value=0.5}
         {tangentMode=RCTM_User, time=0.7, value=0}
        ]}}}
     Scale Curve=&Max_Float(
      A:NiagaraFloat=$System.FN_Timer001.Value
      B:NiagaraFloat=&FloatFromCurve(FloatCurve=DataInterface() {propertyValues {Curve {keys [<at 0 0.2>,<at 0.1 0>]}}}, CurveIndex=$System.LoopedAge, Scale Curve=5)
     )
    )
    Module.Param0WriteEnabled=true
    Module.Param1WriteEnabled=false
    Module.Param2WriteEnabled=true
    Module.Param3WriteEnabled=false
   )
   &Drag(Drag=2, Use Linear Drag=true, Use Rotational Drag=false)
   &InheritSourceMovement(
    Applied Force Scale=<-0.5 -0.5 0>
    ApplyForce=true
    ApplyForceFalloff=false
    ApplyPosition=false
    Module.LimitForceSpeed=false
    Source Velocity:Vector3f=&TransformVector(Vector=$Engine.Owner.Velocity, Source Space=::World, Destination Space=::Simulation)
   )
   &SolveForcesAndVelocity(Clamp Velocity=false, Limit Acceleration=false)
   &ScaleColor(
    Scale Alpha=&FloatFromCurve(
     FloatCurve=DataInterface() {propertyValues {Curve {keys [
         {arriveTangent=-0.3755, interpMode=RCIM_Cubic, leaveTangent=-0.3755, tangentMode=RCTM_User, time=0, value=1}
         {arriveTangent=-1.6012, interpMode=RCIM_Cubic, leaveTangent=-1.6012, tangentMode=RCTM_User, time=1, value=0}
        ]}}}
     Scale Curve=&Max_Float(
      A:NiagaraFloat=$System.FN_Timer001.Value
      B:NiagaraFloat=&FloatFromCurve(
       FloatCurve=DataInterface() {propertyValues {Curve {keys [{arriveTangent=-9.3191, interpMode=RCIM_Cubic, leaveTangent=-9.3191, tangentMode=RCTM_User, time=0, value=1}, <at 0.3 0>]}}}
       CurveIndex=$System.LoopedAge
      )
     )
    )
    ScaleA=true
    ScaleRGB=false
   )
  ]
  rendererClasses [&NiagaraSpriteRendererProperties]
  renderers [{rendererClass=&NiagaraSpriteRendererProperties}]
 }
 Spark_Pop {
  emitterUpdateScript [
   &EmitterState()
   &SpawnBurst_Instantaneous(
    Spawn Probability=&OneMinusFloat(Float=$System.FN_Timer001.Value)
    Use Spawn Probability=true
    UseLoopCountLimit=false
   )
   &SpawnRate(Spawn Probability=$System.FN_Timer001.Value, SpawnRate=$Engine.InverseDeltaTime, Use Spawn Probability=true)
  ]
  particleSpawnScript [
   &InitializeParticle(
    AdjustAlpha=false
    AdjustHue=true
    AdjustSaturation=true
    AdjustValue=true
    Color=<rgba 15 1.5 1.5 1>
    Color Mode=::NewEnumerator3
    Hue Shift Range:Vector2f=<-0.09 0>
    Lifetime=&Multiply_Float(A=$Engine.DeltaTime, B=1.1)
    Position=<0 90 -6>
    Position Mode=::NewEnumerator1
    Saturation Range:Vector2f=<0.6 0.93>
    Sprite Rotation Mode=::NewEnumerator2
    Sprite Size Max=<9 11>
    Sprite Size Min=<6 9>
    Sprite Size Mode=::NewEnumerator2
    UsePositionOffset=false
    Value Range:Vector2f=<0.2 1>
    Write Material Random=true
   )
   /SetVariables(Particles.SubImageIndex=&RandomRangeFloat(Minimum=0, Maximum=4)) {isSetParameters=true}
   /Niagara/Modules/Audio/PlayAudio(
    Audio Player=DataInterface() {propertyValues {
      Attenuation=null
      Concurrency=null
      MaxPlaysPerTick=10
      SoundToPlay=/BRCosmetics/Sounds/Gliders/OrderGuardMale/Glider_OrderGuardMale_Spark_Cue
      bAllowLoopingOneShotSounds=false
      bLimitPlaysPerTick=true
      bOnlyActiveDuringGameplay=false
      bStopWhenComponentIsDestroyed=true
      ConfigurationUserParameter {parameter {name=None, typeDefHandle {registeredTypeIndex=1}, varData []}}
      ParameterNames []
     }}
    Volume=0.7
   )
  ]
  particleUpdateScript [
   &ParticleState()
   &CameraOffset(Camera Offset Amount=11)
  ]
  rendererClasses [&NiagaraSpriteRendererProperties, &NiagaraLightRendererProperties]
  renderers [
   {rendererClass=&NiagaraSpriteRendererProperties}
   {rendererClass=&NiagaraLightRendererProperties, rendererIndex=1}
  ]
 }
 Sparks {
  emitterUpdateScript [
   &EmitterState()
   &SpawnRate(Spawn Probability=$System.FN_Timer001.Value, SpawnRate=222, Use Spawn Probability=true)
   &SpawnRate(Spawn Probability=$System.FN_Timer001.Value, SpawnGroup=1, SpawnRate=55, Use Spawn Probability=true) {module=SpawnRate001}
   &SpawnBurst_Instantaneous(
    Spawn Count=&RandomRangeInt(Minimum=5, Maximum=11)
    Spawn Probability=&OneMinusFloat(Float=$System.FN_Timer001.Value)
    Use Spawn Probability=true
    UseLoopCountLimit=false
   )
  ]
  particleSpawnScript [
   &InitializeParticle(
    Color Mode=::NewEnumerator0
    Lifetime=&RandomRangeFloat(Minimum=0.066, Maximum=0.433)
    Position=&SelectVectorFromArray(
     Direct Array Index=&SpawnGroup()
     Vector Selection Array=DataInterface() {propertyValues {FloatData [<0 88 -7>, <0 -111 -7>]}}
    )
    Position Mode=::NewEnumerator1
    Sprite Size=&Lerp_Vector2(A=<4 7>, B=<3 33>, Alpha=&RandomRangeFloat(Minimum=0, Maximum=1))
    Sprite Size Mode=::NewEnumerator4
    UsePositionOffset=false
    Write Material Random=true
   )
   &AddVelocity(
    Cone Angle=&SelectFloatFromArray(Direct Array Index=&SpawnGroup(), Float Selection Array=DataInterface() {propertyValues {FloatData [3,11]}})
    Cone Axis=&SelectVectorFromArray(
     Direct Array Index=&SpawnGroup()
     Vector Selection Array=DataInterface() {propertyValues {FloatData [<0 1 0.333>, <0 -1 0>]}}
    )
    Module.UseConeFalloff=false
    Module.UseVelDistribution=false
    Override Seed=false
    Velocity Mode=::NewEnumerator2
    Velocity Speed=&RandomRangeFloat(Minimum=444, Maximum=1111)
   )
   /SetVariables(Particles.SubImageIndex=&RandomRangeFloat(Minimum=0, Maximum=4)) {isSetParameters=true}
   &PlayAudio(
    Audio Player=DataInterface() {propertyValues {
      Attenuation=null
      Concurrency=null
      MaxPlaysPerTick=10
      SoundToPlay=&Glider_OrderGuardMale_Spark_Cue
      bAllowLoopingOneShotSounds=false
      bLimitPlaysPerTick=true
      bOnlyActiveDuringGameplay=false
      bStopWhenComponentIsDestroyed=true
      ConfigurationUserParameter {parameter {name=None, typeDefHandle {registeredTypeIndex=1}, varData []}}
      ParameterNames []
     }}
    Volume=0.7
   )
  ]
  particleUpdateScript [
   &Drag(Drag=&RandomRangeFloat(Minimum=1, Maximum=2), Use Linear Drag=true, Use Rotational Drag=false)
   &GravityForce(Gravity=<0 0 -981>)
   &InheritSourceMovement(
    Applied Force Scale=<-3 -3 0>
    ApplyForce=true
    ApplyForceFalloff=false
    ApplyPosition=false
    Module.LimitForceSpeed=false
    Source Velocity:Vector3f=&TransformVector(Vector=$Engine.Owner.Velocity, Source Space=::World, Destination Space=::Simulation)
   )
   &SolveForcesAndVelocity(Clamp Velocity=false, Limit Acceleration=false)
   &ParticleState()
   &ScaleSpriteSize(
    Non-Uniform Curve Sprite Scale=DataInterface() {propertyValues {
      XCurve {keys [{tangentMode=RCTM_User, time=0, value=1}, <at 0.6 1>, {tangentMode=RCTM_User, time=1, value=0}]}
      YCurve {keys [{tangentMode=RCTM_User, time=0, value=1}, {tangentMode=RCTM_User, time=1, value=0}]}
     }}
    Scale Sprite Size Mode=::NewEnumerator3
   )
   &Color(
    Color=&ColorFromCurve(ColorCurve=DataInterface() {propertyValues {
       AlphaCurve {keys [<at 0 0 RCIM_Cubic>,<at 0.1142 1>,<at 1 0>]}
       BlueCurve {keys [<at 0 6.93 RCIM_Cubic>, <at 0.1579 0.26>, {interpMode=RCIM_Cubic, leaveTangent=-0.0432, time=0.2588, value=0}]}
       GreenCurve {keys [<at 0 31.482 RCIM_Cubic>, <at 0.1579 1.7463>, {interpMode=RCIM_Cubic, leaveTangent=-0.5416, time=0.2588, value=0.3}]}
       RedCurve {keys [<at 0 99 RCIM_Cubic>, <at 0.1579 13>, {interpMode=RCIM_Cubic, leaveTangent=-4.3157, time=0.2588, value=9}]}
      }})
    Scale Alpha=&SelectFloatFromArray(Direct Array Index=&SpawnGroup(), Float Selection Array=DataInterface() {propertyValues {FloatData [1,5]}})
    Scale Color=&VectorFromFloat(Value=&SelectFloatFromArray(Direct Array Index=&SpawnGroup(), Float Selection Array=DataInterface() {propertyValues {FloatData [1,5]}}))
   )
   &ScaleSpriteSizeBySpeed(Max Scale Factor=<1 3>, Velocity Threshold=999)
  ]
  rendererClasses [&NiagaraSpriteRendererProperties]
  renderers [{rendererClass=&NiagaraSpriteRendererProperties}]
 }
}
summary {
 systemName=NS_Glider_OrderGuard
 emitters [
  {bEnabled=false, emitterName=FN_Debug, rendererClasses [&NiagaraSpriteRendererProperties]}
  {emitterName=Sparks, rendererClasses [&NiagaraSpriteRendererProperties]}
  {emitterName=Meshes, rendererClasses [&NiagaraMeshRendererProperties]}
  {emitterName=Spark_Pop, rendererClasses [&NiagaraSpriteRendererProperties, &NiagaraLightRendererProperties]}
  {emitterName=Smoke, rendererClasses [&NiagaraSpriteRendererProperties]}
  {emitterName=Electricity, rendererClasses [&NiagaraMeshRendererProperties, &NiagaraSpriteRendererProperties]}
 ]
 userVariables [
  User.bIsFrontEnd {type=NiagaraBool}
  User.bIsFrontEndPreview {type=NiagaraBool}
  User.bIsFullyDeployed {type=NiagaraBool}
  User.ForwardDot {type=NiagaraFloat, defaultValue=0, description=""}
  User.LightningOrder {
   type=NiagaraDataInterfaceArrayInt32
   description=""
   defaultValue {
    struct=NiagaraExt_VariableValue_DataInterface
    value {
     dataInterface=&self:NiagaraDataInterfaceArrayInt32_0
     dataInterfaceClass=&NiagaraDataInterfaceArrayInt32
    }
   }
  }
  User.RightDot {type=NiagaraFloat, defaultValue=0, description=""}
  User.SkeletalMesh {
   type=NiagaraDataInterfaceSkeletalMesh
   defaultValue {
    struct=NiagaraExt_VariableValue_DataInterface
    value {dataInterface=&self:NiagaraDataInterfaceSkeletalMesh_0}
   }
  }
 ]
}
